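# Neon tube material for the holiday llama prop: a glossy, metallic-by-default glow shaped by an axis-aligned Fresnel.
# Color is either a single parameter or a palette selected by RGB vertex-color masks, each entry scaled by random-mask parameters.
# An optional flicker thresholds a time-scrolled texture, offset by object position, alongside a LightsOn control.
# Ray tracing swaps Fresnel for a Lumen emissive multiplier, while brightness, exponent and roughness have mobile-path variants.
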Material /Game/Environments/Asteria/Props/Holiday/Holiday_Llama_Neon_A/Materials/M_NeonTube
graph {
 properties {bUsedWithNanite=true, pixelDepthOffsetMode=PDOM_Legacy, usedShadingModels=Default Lit}
 nodes [
  Constant3Vector() {constant=<rgba 1 1 1 0>}
  n1 = ScalarParameter() {defaultValue=10, parameterName=GlowFresnelExponent}
  n2 = Multiply(A=@n1, B=ScalarParameter() {defaultValue=1, parameterName=GlowFresnelExponentScale_Mobile})
  n3 = Multiply(A=@n1, B=ScalarParameter() {defaultValue=1, parameterName=GlowFresnelExponentScale_Switch})
  n4 = MaterialFunctionCall!(
   Axis=Transform(None=Constant3Vector() {constant=<rgba 0 1 0 0>})
   Exponent=ShadingPathSwitch(Default=@n1, Deferred=@n1, Mobile=QualitySwitch(Default=@n2, Low=@n2, Medium=@n3))
   Normal
  ) {materialFunction=/Game/Effects/Fort_Effects/Materials/_Functions/MF_AxisAlignedFrensnel}
  n5 = ScalarParameter() {defaultValue=0.1, parameterName=LumenEmissiveMult}
  n6 = ObjectPositionWS()
  n7 = Divide(
   A=Add(
    A=Time()
    B=MaterialFunctionCall(Vector 3=@n6) {materialFunction=/Engine/Functions/Engine_MaterialFunctions02/Utility/VectorLength}.V3 Length
   )
   B=Divide(B=ScalarParameter() {defaultValue=1, parameterName=FlickerSpeed}) {constA=512, constB=512}
  ) {constB=512}
  n8 = ScalarParameter() {defaultValue=1, parameterName=LightsOn, sliderMax=1}
  n9 = Multiply(
   A=If(
    A=TextureSample(UVs=@n7) {samplerType=SAMPLERTYPE_LinearColor, texture=/Game/Environments/Asteria/Props/Holiday/Holiday_Llama_Neon_A/Textures/T_Flicker_A}.R
    B=ScalarParameter() {defaultValue=0.75, parameterName=FlickerTreshold, sliderMax=1}
    "A > B"=Constant()
    "A < B"=Constant() {r=1}
   )
   B=@n8
  )
  n10 = VectorParameter() {parameterName=FlickerRandomMask_A, defaultValue=<rgba 1 1 1 1>}
  n11 = VertexColor()
  n12 = LinearInterpolate(
   A=Multiply(A=@n10.R, B=VectorParameter() {desc=Warm White, parameterName=Color_None, defaultValue=<rgba 1 0.8495 0.682 0>}.RGB)
   B=Multiply(A=@n10.G, B=VectorParameter() {desc=Light blue, parameterName=Color_B, defaultValue=<rgba 0.0142 0.3663 1 0>}.RGB)
   Alpha=@n11.B
  ) {desc=Blue}
  n13 = LinearInterpolate(
   A=Multiply(A=@n10.B, B=VectorParameter() {desc=Magenta, parameterName=Color_G, defaultValue=<rgba 1 0.8495 0.682 0>}.RGB)
   B=Multiply(A=@n10.A, B=VectorParameter() {desc=Peach, parameterName=Color_GB, defaultValue=<rgba 1 0.5904 0.6403 0>}.RGB)
   Alpha=@n11.B
  ) {desc=Blue}
  n14 = VectorParameter() {parameterName=FlickerRandomMask_B, defaultValue=<rgba 1 1 1 1>}
  n15 = LinearInterpolate(
   A=Multiply(A=@n14.R, B=VectorParameter() {desc=Turquoise, parameterName=Color_R, defaultValue=<rgba 0 1 0.6162 0>}.RGB)
   B=Multiply(A=@n14.G, B=VectorParameter() {desc=Red, parameterName=Color_RB, defaultValue=<rgba 1 0.0982 0.0285 0>}.RGB)
   Alpha=@n11.B
  ) {desc=Blue}
  n16 = Multiply(A=@n14.B, B=VectorParameter() {desc=Yellow, parameterName=Color_RG, defaultValue=<rgba 1 0.8702 0 0>}.RGB)
  n17 = StaticSwitchParameter(
   True=LinearInterpolate(A=LinearInterpolate(A=@n12, B=@n13, Alpha=@n11.G) {desc=Green}, B=LinearInterpolate(A=@n15, B=@n16, Alpha=@n11.G) {desc=Green}, Alpha=@n11.R) {desc=Red}
   False=VectorParameter() {parameterName=Color, defaultValue=<rgba 1 0 0.4825 1>}.RGB
  ) {parameterName=Use VertexColor as channel masks}
  n18 = Multiply(A=@n17, B=ScalarParameter() {defaultValue=25, parameterName=Brightness})
  n19 = Multiply(A=@n17, B=ScalarParameter() {defaultValue=10, parameterName=Brightness_Mobile})
  n20 = Multiply(A=@n17, B=ScalarParameter() {defaultValue=10, parameterName=Brightness_Switch})
  Multiply(
   A=Multiply(A=RayTracingQualitySwitch(Normal=@n4.Fresnel, RayTraced=@n5), B=StaticSwitchParameter(True=@n9, False=@n8) {parameterName=Flicker})
   B=ShadingPathSwitch(Default=@n18, Deferred=@n18, Mobile=QualitySwitch(Default=@n19, Low=@n19, Medium=@n20))
  )
  ScalarParameter() {defaultValue=1, parameterName=Metallic, sliderMax=1}
  n21 = ScalarParameter() {defaultValue=0.148, parameterName=Roughness, sliderMax=1}
  ShadingPathSwitch(Default=@n21, Deferred=@n21, Mobile=ScalarParameter() {defaultValue=0.948, parameterName=Roughness_Mobile, sliderMax=1})
 ]
}
