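# Default Lit material for a simple neon tube, filed with the Holiday Llama Neon prop in the Asteria environment.
# A Color parameter is multiplied by an inverted low-exponent Fresnel, weighting the tint toward view-facing surfaces.
# A Quality Switch scales that result by the very large MobileBump parameter on Default, Low and Medium, and by a small constant on High and Epic.
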Material /Game/Environments/Asteria/Props/Holiday/Holiday_Llama_Neon_A/Materials/M_SimpleNeonTube
graph {
 properties {bAutomaticallySetUsageInEditor=false, usedShadingModels=Default Lit}
 nodes [
  Constant3Vector()
  Constant() {r=0.66}
  Constant() {r=0.1}
  n1 = ScalarParameter() {defaultValue=8000, parameterName=MobileBump}
  n2 = Constant() {r=0.5}
  n3 = Fresnel() {baseReflectFraction=0.01, exponent=0.1}
  Multiply(
   A=QualitySwitch(Default=@n1, Low=@n1, High=@n2, Medium=@n1, Epic=@n2)
   B=Multiply(A=VectorParameter() {parameterName=Color, defaultValue=<rgba 1 0 0.4825 1>}.RGB, B=OneMinus(None=@n3))
  )
 ]
}
